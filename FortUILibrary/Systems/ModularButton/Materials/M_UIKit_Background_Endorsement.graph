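# Translucent UI material for the background of the Endorsement variant of the UIKit modular button.
# A Slate post buffer sample provides the blur color for two CoreUI backplate functions, standard (A) and disabled.
# A button-states lerp function builds the interaction color, with hovered, focused and pressed sharing one input and disabled using another.
# A CoreUI rounded mask function exposes corner radius parameters, including one in pixels.
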Material /FortUILibrary/Systems/ModularButton/Materials/M_UIKit_Background_Endorsement
graph {
 properties {
  bUsedWithStaticMesh=false
  blendMode=BLEND_Translucent
  materialDomain=MD_UI
  pixelDepthOffsetMode=PDOM_Legacy
  usedShadingModels=Default Lit
 }
 nodes [
  MaterialFunctionCall(
   Corner Radius=ScalarParameter() {defaultValue=0.3, parameterName=Corner Radius}
   Radius in Pixels=ScalarParameter() {parameterName=Radius in Pixels}
  ) {materialFunction=&MF_UI_CoreUI_RoundedMask}
  NamedRerouteDeclaration(None=MaterialFunctionCall!("UVs (Optional)") {materialFunction=/Engine/Functions/UserInterface/GetSlatePost0}.LinearRGB) {name=Blur RGB, nodeColor=<rgba 0 0 0 1>}
  NamedRerouteDeclaration(None=MaterialFunctionCall!(Post Buffer Blur=NamedRerouteUsage()) {materialFunction=/FortUILibrary/MaterialFunctions/CoreUI/MF_UI_CoreUI_BackplateA}.Result) {name=Backplate A, nodeColor=<rgba 0 0 0 1>}
  n1 = NamedRerouteUsage()
  NamedRerouteDeclaration(None=MaterialFunctionCall!(
    Idle
    Idle Opacity
    Hovered=@n1
    Hovered Opacity
    Focused=@n1
    Focused Opacity
    Pressed=@n1
    Pressed Opacity
    Disabled=NamedRerouteUsage()
    Disabled Opacity
   ) {materialFunction=/FortUILibrary/Systems/ModularButton/MaterialFunctions/MF_UIKit_ButtonStatesBase_Lerp}.Color) {name=Interaction Color, nodeColor=<rgba 0 0 0 1>}
  NamedRerouteDeclaration(None=MaterialFunctionCall!(Post Buffer Blur=NamedRerouteUsage()) {materialFunction=/FortUILibrary/MaterialFunctions/CoreUI/MF_UI_CoreUI_BackplateDisabled}.Result) {name=Backplate Disabled, nodeColor=<rgba 0 0 0 1>}
  NamedRerouteUsage()
 ]
}
