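# Translucent UI material serving as the fallback background for modular button CTAs, using flat constant colors and a rounded-corner mask at half strength.
# Display Regular and Display Loud parameters blend between Quiet, Regular and yellow Loud styles.
# The styled color feeds Idle, Hovered, Focused and Pressed alike in the button-state lerp function, with Disabled wired separately.
# An Is Selectable static switch routes that result through a selectable-state lerp to form the interaction color.
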Material /FortUILibrary/Systems/ModularButton/Materials/M_UIKit_CTA_Background_Fallback
graph {
 properties {
  bUsedWithStaticMesh=false
  blendMode=BLEND_Translucent
  materialDomain=MD_UI
  pixelDepthOffsetMode=PDOM_Legacy
  usedShadingModels=Default Lit
 }
 nodes [
  Multiply(A=MaterialFunctionCall(
    Corner Radius=ScalarParameter() {defaultValue=0.3, parameterName=Corner Radius}
    Radius in Pixels=ScalarParameter() {parameterName=Radius in Pixels}
   ) {materialFunction=&MF_UI_CoreUI_RoundedMask}.Shape Mask) {constB=0.5}
  NamedRerouteDeclaration(None=Constant3Vector() {constant=<rgba 0.7084 0.6867 0.8714 1>}) {name=Backplate A, nodeColor=<rgba 0 0 0 1>}
  NamedRerouteDeclaration(None=Constant3Vector() {constant=<rgba 0.3864 0.3763 0.4793 1>}) {name=Backplate B, nodeColor=<rgba 0 0 0 1>}
  n1 = NamedRerouteUsage()
  n2 = NamedRerouteUsage()
  n3 = LinearInterpolate(
   A=LinearInterpolate(
    A=NamedRerouteDeclaration(None=@n1) {name=Quiet CTA, nodeColor=<rgba 0 0 0 1>}
    B=NamedRerouteDeclaration(None=@n2) {name=Regular CTA, nodeColor=<rgba 0 0 0 1>}
    Alpha=ScalarParameter() {defaultValue=1, group=Style, parameterName=Display Regular, sliderMax=1}
   )
   B=NamedRerouteDeclaration(None=Constant3Vector() {constant=<rgba 0.9301 1 0.0103 1>}) {name=Loud CTA, nodeColor=<rgba 0 0 0 1>}
   Alpha=ScalarParameter() {defaultValue=1, group=Style, parameterName=Display Loud, sliderMax=1}
  )
  n4 = MaterialFunctionCall!(
   Idle=@n3
   Idle Opacity
   Hovered=@n3
   Hovered Opacity
   Focused=@n3
   Focused Opacity
   Pressed=@n3
   Pressed Opacity
   Disabled=NamedRerouteUsage()
   Disabled Opacity
  ) {materialFunction=/FortUILibrary/Systems/ModularButton/MaterialFunctions/MF_UIKit_ButtonStatesBase_Lerp}
  NamedRerouteDeclaration(None=StaticSwitchParameter(
    True=MaterialFunctionCall!(
     Base States=@n4.Color
     Base States Opacity
     Selected_Idle
     Selected_Idle Opacity
     Selected_Hovered
     Selected_Hovered Opacity
     Selected_Focused
     Selected_Focused Opacity
     Selected_Pressed
     Selected_Pressed Opacity
     Selected_Disabled
     Selected_Disabled Opacity
    ) {materialFunction=/FortUILibrary/Systems/ModularButton/MaterialFunctions/MF_UIKit_ButtonStatesSelectable_Lerp}.Color
    False=@n4.Color
   ) {parameterName=Is Selectable}) {name=Interaction Color, nodeColor=<rgba 0 0 0 1>}
  NamedRerouteDeclaration(None=Constant3Vector() {constant=<rgba 0.0513 0.0513 0.0513 1>}) {name=Backplate Disabled, nodeColor=<rgba 0 0 0 1>}
  NamedRerouteUsage()
 ]
}
