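# M_UIKit_Color is a translucent UI-domain material for the modular button system in FortUILibrary, named for color-based button styling.
# Its graph holds one call to MF_UIKit_ButtonStatesBase_Lerp, which exposes color and opacity pairs for Idle, Hovered, Focused, Pressed and Disabled states.
# The material adds no parameters, textures or math of its own, leaving any state blending to that shared function.
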Material /FortUILibrary/Systems/ModularButton/Materials/M_UIKit_Color
graph {
 properties {
  bUsedWithStaticMesh=false
  blendMode=BLEND_Translucent
  materialDomain=MD_UI
  pixelDepthOffsetMode=PDOM_Legacy
  usedShadingModels=Default Lit
 }
 nodes [MaterialFunctionCall!(Idle, Idle Opacity, Hovered, Hovered Opacity, Focused, Focused Opacity, Pressed, Pressed Opacity, Disabled, Disabled Opacity) {materialFunction=/FortUILibrary/Systems/ModularButton/MaterialFunctions/MF_UIKit_ButtonStatesBase_Lerp}]
}
