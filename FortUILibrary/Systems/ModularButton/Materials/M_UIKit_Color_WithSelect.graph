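# UI-domain translucent material for the UIKit Modular Button, handling color and opacity across base and selected button states.
# A base lerp function covers Idle, Hovered, Focused, Pressed and Disabled, and feeds a selectable lerp function that adds Selected variants of those states.
# The base function's Color output drives both Base States and Base States Opacity on the selectable function.
# The per-state color and opacity inputs are left unwired.
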Material /FortUILibrary/Systems/ModularButton/Materials/M_UIKit_Color_WithSelect
graph {
 properties {
  bUsedWithStaticMesh=false
  blendMode=BLEND_Translucent
  materialDomain=MD_UI
  pixelDepthOffsetMode=PDOM_Legacy
  usedShadingModels=Default Lit
 }
 nodes [
  n1 = MaterialFunctionCall!(Idle, Idle Opacity, Hovered, Hovered Opacity, Focused, Focused Opacity, Pressed, Pressed Opacity, Disabled, Disabled Opacity) {materialFunction=/FortUILibrary/Systems/ModularButton/MaterialFunctions/MF_UIKit_ButtonStatesBase_Lerp}
  MaterialFunctionCall!(
   Base States=@n1.Color
   Base States Opacity=@n1.Color
   Selected_Idle
   Selected_Idle Opacity
   Selected_Hovered
   Selected_Hovered Opacity
   Selected_Focused
   Selected_Focused Opacity
   Selected_Pressed
   Selected_Pressed Opacity
   Selected_Disabled
   Selected_Disabled Opacity
  ) {materialFunction=/FortUILibrary/Systems/ModularButton/MaterialFunctions/MF_UIKit_ButtonStatesSelectable_Lerp}
 ]
}
