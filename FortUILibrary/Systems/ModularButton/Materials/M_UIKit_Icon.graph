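# M_UIKit_Icon is a translucent UI material for icons in the modular button system.
# It samples an Icon texture parameter and takes the maximum of its red and blue channels; green is unused, as the first Max compares red with itself.
# That value is multiplied by the Opacity output of MF_UIKit_ButtonStatesBase_Lerp, which covers idle, hovered, focused, pressed and disabled states, then by the texture's alpha.
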Material /FortUILibrary/Systems/ModularButton/Materials/M_UIKit_Icon
graph {
 properties {
  bUsedWithStaticMesh=false
  blendMode=BLEND_Translucent
  materialDomain=MD_UI
  pixelDepthOffsetMode=PDOM_Legacy
  usedShadingModels=Default Lit
 }
 nodes [
  n1 = TextureSampleParameter2D() {parameterName=Icon}
  n2 = Max(A=@n1.R, B=@n1.R)
  Multiply(
   A=Multiply(
    A=MaterialFunctionCall!(Idle, Idle Opacity, Hovered, Hovered Opacity, Focused, Focused Opacity, Pressed, Pressed Opacity, Disabled, Disabled Opacity) {materialFunction=/FortUILibrary/Systems/ModularButton/MaterialFunctions/MF_UIKit_ButtonStatesBase_Lerp}.Opacity
    B=Max(A=@n2, B=@n1.B)
   )
   B=@n1.A
  )
 ]
}
