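# Stereo one-shot MetaSound source for the release sound of a regular holdable button in the front-end UI.
# On play, the nested MSS_PlayRandomSound_Stereo graph plays from the Sounds array, which holds a single release wave by default, and passes audio and the finished trigger straight out.
# It routes to the Fort_UI sound class at reduced volume, without virtualization, and a concurrency override lets only the newest instance play.
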MetaSoundSource /Game/Sounds/Fort_UI_Sounds/FrontEnd/DefaultSounds/HoldableButtons/Holdable_Regular/MSS_UI_Button_Holdable_Regular_Release
inputs [
 Sounds {type=WaveAsset:Array}
 UE.Source.OnPlay {type=Trigger}
]
interfaces [UE.OutputFormat.Stereo:1.0, UE.Source.OneShot:1.0, UE.Source:1.1]
outputs [
 "UE.OutputFormat.Stereo.Audio:0" {type=Audio}
 "UE.OutputFormat.Stereo.Audio:1" {type=Audio}
 UE.Source.OneShot.OnFinished {type=Trigger}
]
pages [Default {
  input_defaults {UE.Source.OnPlay=null, Sounds [/Game/Sounds/Fort_UI_Sounds/FrontEnd/DefaultSounds/HoldableButtons/wav/UI_Button_Press_Hold_Normal_Release]}
  nodes [MSS_PlayRandomSound_Stereo_0 = &MSS_PlayRandomSound_Stereo(Sounds=^Sounds, UE.Source.OnPlay=^UE.Source.OnPlay)]
  links [
   {dst="@UE.OutputFormat.Stereo.Audio:0", src=MSS_PlayRandomSound_Stereo_0.UE.OutputFormat.Stereo.Audio:0}
   {dst="@UE.OutputFormat.Stereo.Audio:1", src=MSS_PlayRandomSound_Stereo_0.UE.OutputFormat.Stereo.Audio:1}
   {dst="@UE.Source.OneShot.OnFinished", src=MSS_PlayRandomSound_Stereo_0.UE.Source.OneShot.OnFinished}
  ]
 }]
properties {
 assetImportData=&self:AssetImportData
 bOverrideConcurrency=true
 numChannels=2
 outputFormat=Stereo
 soundClassObject=&Fort_UI
 virtualizationMode=Disabled
 volume=0.378
 concurrencyOverrides {maxCount=1, resolutionRule=StopOldest}
}
