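# Base UIKit modular button widget that assembles optional text, second text, icon, background and status indicator blocks, with an input action prompt.
# PreConstruct creates and registers the blocks in grid or stack slots, mirroring alignments and padding for Arabic.
# Button events drive timed or instant state transitions across blocks and play hover, press and hold sounds.
# Hold progress feeds a material parameter, and enable changes apply one tick later.
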WidgetBlueprint /FortUILibrary/Systems/ModularButton/WBP_UIKit_ButtonGeneric_Base
animations [Anim_Focused_Base {
  binding_count=1
  length_seconds=0.25
  object=&self:Anim_Focused_Base
  sequence {bindings [Prompt {
     id=#1
     possessed_class=GridPanel
     tracks [{
       type=MovieScene2DTransformTrack
       sections [{
         type=MovieScene2DTransformSection
         channels [
          Translation.X {type=MovieSceneScriptingFloatChannel, keys [[2,0],[5,-6]]}
          Translation.Y {type=MovieSceneScriptingFloatChannel}
          Scale.X {type=MovieSceneScriptingFloatChannel, keys [[0,1],[1.3333,0.9],[3,1.125],[5.0003,1.05]]}
          Scale.Y {type=MovieSceneScriptingFloatChannel, keys [[0,1],[1.3333,0.9],[3,1.125],[5.0003,1.05]]}
         ]
        }]
      }]
    }]}
 }]
graphs [
 AddBackground {nodes [
   AddBackground() {is=title K2Node_FunctionEntry}
   > RemoveWidget!(execute, Widget:Widget Object Reference=Get Block_Background() -> (Block_Background:WBP UIKit Block Background Base Object Reference)) -> (then)
   > Branch(Condition=Get UseBackgroundBlock() -> (UseBackgroundBlock:bool))
   > n1 = Create Widget!(
    execute:Exec
    Class:User Widget Class Reference=Get BackgroundClass() -> (BackgroundClass:WBP UIKit Block Background Base Class Reference)
    OwningPlayer:Player Controller Object Reference
   ) -> (then:Exec, ReturnValue:WBP UIKit Block Background Base Object Reference)
   > n2 = Set Block_Background(Block_Background:WBP UIKit Block Background Base Object Reference=@n1) -> (Output_Get:WBP UIKit Block Background Base Object Reference)
   > Branch(Condition=IsValid(Object=@n2))
   > RegisterModularBlock!(execute, self:UIKit Modular Button Object Reference, Block:UIKit Block Object Reference=@n2) -> (then)
   > SetIsSelectable!(execute, self:WBP UIKit Block Base Object Reference=@n2, Is Selectable:bool=Get bSelectable() -> (bSelectable:bool)) -> (then)
   n3 = Get Root() -> (Root:Grid Panel Object Reference)
   > n4 = Is Valid(InputObject=@n3)
   n5 = AND Boolean(A=Get UseTextBlock() -> (UseTextBlock:bool), B=Get IsTextOutsideBackground() -> (IsTextOutsideBackground:bool))
   n6 = Get TextBlockAlignment() -> (TextBlockAlignment:E_UI_Alignment Enum=NewEnumerator0)
   > n7 = AddChildToGrid!(
    execute
    self:Grid Panel Object Reference=@n3
    Content:Widget Object Reference=@n2
    InRow:int=SelectInt(
     A=Select!(NewEnumerator0:int=2, NewEnumerator1:int=0, NewEnumerator2:int=1, NewEnumerator3:int=1, Index:E_UI_Alignment Enum=@n6) -> (ReturnValue:int)
     B=1
     bPickA=@n5
    )
    InColumn:int=SelectInt(A=Select(NewEnumerator0=1, NewEnumerator1=1, NewEnumerator2=2, NewEnumerator3=0, Index=@n6), B=1, bPickA=@n5)
   ) -> (then, ReturnValue:Grid Slot Object Reference)
   SetGridSlotLayer!(execute=[@n7.then, @n4.Is Not Valid], Widget:Widget Object Reference=@n2, Layer:int=1) -> (then)
  ]}
 AddIconBlock {nodes [
   AddIconBlock() {is=title K2Node_FunctionEntry}
   > n1 = Get!(execute:Exec) -> (then:Exec, else:Exec, Block_Icon:WBP UIKit Block Image Object Reference)
   n2 = Get Block_Icon() -> (Block_Icon:WBP UIKit Block Image Object Reference)
   RegisterModularBlock(execute=[@n1.then, @n14.then], Block=@n2)
   > SetIsSelectable(self=@n2, Is Selectable=Get bSelectable())
   n3 = Get Main() -> (Main:Grid Panel Object Reference)
   > n4 = Is Valid(InputObject=@n3)
   n5 = AND Boolean(A=Get UseTextBlock(), B=NOT Boolean(A=Get IsTextOutsideBackground()))
   n6 = Get TextBlockAlignment() -> (TextBlockAlignment=NewEnumerator0)
   n7 = Select(NewEnumerator0=2, NewEnumerator1=0, NewEnumerator2=1, NewEnumerator3=1, Index=@n6)
   n8 = AND Boolean(A=Get UseSecondTextBlock() -> (UseSecondTextBlock:bool), B=Get IsTextOutsideBackground())
   n9 = Get SecondTextBlockAlignment() -> (SecondTextBlockAlignment:E_UI_Alignment Enum=NewEnumerator0)
   n10 = Select(NewEnumerator0=1, NewEnumerator1=1, NewEnumerator2=2, NewEnumerator3=0, Index=@n6)
   > n11 = AddChildToGrid(
    self=@n3
    Content=@n2
    InRow=SelectInt(A=Select(NewEnumerator0=2, NewEnumerator1=0, NewEnumerator2=1, NewEnumerator3=1, Index=@n9), B=SelectInt(A=@n7, B=1, bPickA=@n5), bPickA=@n8)
    InColumn=SelectInt(
     A=Select(NewEnumerator0=1, NewEnumerator1=1, NewEnumerator2=2, NewEnumerator3=0, Index=@n9)
     B=SelectInt(A=@n10, B=1, bPickA=@n5)
     bPickA=@n8
    )
   )
   SetGridSlotLayer(execute=[@n11.then, @n4.Is Not Valid], Widget=@n2, Layer=10)
   > SetSize!(execute, self:WBP UIKit Block Image Object Reference=@n2, Desired Size:Vector 2D Structure=Get IconSize() -> (IconSize:Vector 2D Structure)) -> (then)
   > SetGridSlotAlignment!(
    execute
    Widget:Widget Object Reference=@n2
    Horizontal:EHorizontalAlignment Enum=Get IconHorizontalAlignment() -> (IconHorizontalAlignment:EHorizontalAlignment Enum=HAlign_Fill)
    Vertical:EVerticalAlignment Enum=Get IconVerticalAlignment() -> (IconVerticalAlignment:EVerticalAlignment Enum=VAlign_Fill)
   ) -> (then)
   > SetGridSlotPadding!(execute, Widget:Widget Object Reference=@n2, Padding:Margin Structure=Get IconPadding() -> (IconPadding:Margin Structure)) -> (then)
   n12 = Get IconClass() -> (IconClass:WBP UIKit Block Image Class Reference)
   n13 = Create Widget(
    execute=@n1.else
    Class=Select!(
     Option 0:WBP UIKit Block Image Class Reference=/FortUILibrary/Systems/ModularButton/Blocks/WBP_UIKit_Block_Image.WBP_UIKit_Block_Image_C
     Option 1:WBP UIKit Block Image Class Reference=@n12
     Index:bool=IsValidClass!(Class:Object Class Reference=@n12) -> (ReturnValue:bool)
    ) -> (ReturnValue:WBP UIKit Block Image Class Reference)
    OwningPlayer=GetOwningPlayer()
   ) -> (ReturnValue:WBP UIKit Block Image Object Reference)
   > n14 = Set Block_Icon(Block_Icon:WBP UIKit Block Image Object Reference=@n13) -> (Output_Get:WBP UIKit Block Image Object Reference)
  ]}
 AddSecondTextBlock {nodes [
   AddSecondTextBlock() {is=title K2Node_FunctionEntry}
   > RemoveWidget(Widget=Get Block_SecondText() -> (Block_SecondText:WBP UIKit Block Text Object Reference))
   > Branch(Condition=Get UseSecondTextBlock())
   > n1 = Create Widget(Class=Get SecondTextClass() -> (SecondTextClass:WBP UIKit Block Text Class Reference)) -> (ReturnValue:WBP UIKit Block Text Object Reference)
   > n2 = Set Block_SecondText(Block_SecondText:WBP UIKit Block Text Object Reference=@n1) -> (Output_Get:WBP UIKit Block Text Object Reference)
   > Branch(Condition=IsValid(Object=@n2))
   > RegisterModularBlock(Block=@n2)
   > SetIsSelectable(self=@n2, Is Selectable=Get bSelectable())
   > AddChildToStackBox!(execute, self:Stack Box Object Reference=Get Main_Stack() -> (Main_Stack:Stack Box Object Reference), Content:Widget Object Reference=@n2) -> (then, ReturnValue:Stack Box Slot Object Reference)
  ]}
 AddStatusIndicator {nodes [
   AddStatusIndicator() {is=title K2Node_FunctionEntry}
   > RemoveWidget(Widget=Get StatusIndicator() -> (StatusIndicator:WBP UIKit Status Indicator Object Reference))
   > Branch(Condition=Get UseStatusIndicator() -> (UseStatusIndicator:bool))
   > n1 = Create Widget(Class=Get StatusIndicatorClass() -> (StatusIndicatorClass:WBP UIKit Status Indicator Class Reference)) -> (ReturnValue:WBP UIKit Status Indicator Object Reference)
   > n2 = Set StatusIndicator(StatusIndicator:WBP UIKit Status Indicator Object Reference=@n1) -> (Output_Get:WBP UIKit Status Indicator Object Reference)
   > Branch(Condition=IsValid(Object=@n2))
   > n3 = Branch(Condition=Get IsStatusIndicatorInsideMainGrid() -> (IsStatusIndicatorInsideMainGrid:bool))
   > n4 = AddChildToGrid(self=Get Main(), Content=@n2, InRow=1, InColumn=2)
   SetGridSlotLayer(execute=[@n7.then, @n4.then], Widget=@n2, Layer=11)
   n5 = AND Boolean(A=Get UseTextBlock(), B=Get IsTextOutsideBackground())
   n6 = Get TextBlockAlignment() -> (TextBlockAlignment=NewEnumerator0)
   n7 = AddChildToGrid(
    execute=@n3.else
    self=Get Root()
    Content=@n2
    InRow=SelectInt(
     A=Select!(NewEnumerator0:int=2, NewEnumerator1:int=0, NewEnumerator2:int=1, NewEnumerator3:int=1, Index:E_UI_Alignment Enum=@n6) -> (ReturnValue:int)
     B=1
     bPickA=@n5
    )
    InColumn=SelectInt(A=Select(NewEnumerator0=1, NewEnumerator1=1, NewEnumerator2=2, NewEnumerator3=0, Index=@n6), B=1, bPickA=@n5)
   )
  ]}
 AddTextBlock {nodes [
   AddTextBlock() {is=title K2Node_FunctionEntry}
   > RemoveWidget(Widget=Get Block_Text() -> (Block_Text:WBP UIKit Block Text Object Reference))
   > Branch(Condition=Get UseTextBlock())
   > n1 = Create Widget(Class=Get TextClass() -> (TextClass:WBP UIKit Block Text Class Reference)) -> (ReturnValue:WBP UIKit Block Text Object Reference)
   > n2 = Set Block_Text(Block_Text:WBP UIKit Block Text Object Reference=@n1) -> (Output_Get:WBP UIKit Block Text Object Reference)
   > Branch(Condition=IsValid(Object=@n2))
   > RegisterModularBlock(Block=@n2)
   > SetIsSelectable(self=@n2, Is Selectable=Get bSelectable())
   > n3 = Branch(Condition=Get ShouldTextImpactLayout() -> (ShouldTextImpactLayout:bool))
   > n4 = Branch(Condition=Get IsTextOutsideBackground())
   > AddChildToGrid(self=Get Root(), Content=@n2, InRow=1, InColumn=1)
   n5 = Get Main_Stack()
   Is Valid(exec=@n4.else, InputObject=@n5)
   > AddChildToStackBox(self=@n5, Content=@n2)
   AddChild(execute=@n3.else, self=Get VisualAttachment() -> (VisualAttachment:Common Visual Attachment Object Reference), Content=@n2)
  ]}
 ClearPrompt {nodes [
   ClearPrompt() {is=title K2Node_FunctionEntry}
   > SetVisibility(
    self=[
     Get HoldProgressBar() -> (HoldProgressBar:Image Object Reference)
     Get Prompt() -> (Prompt:Grid Panel Object Reference)
     Get HoldIcon() -> (HoldIcon:Image Object Reference)
    ]
    InVisibility=Collapsed
   )
   > SetHoldPercentage!(execute, self, HoldPercentage:double=0) -> (then)
  ]}
 EventGraph {nodes [
   "On Input Icon Updated (InputActionWidget)"() {is=title K2Node_ComponentBoundEvent}
   > SetInputIconDarkMode!(execute, self, UseDark:bool=Get UseDarkInputIcon() -> (UseDarkInputIcon:bool)) -> (then)
   n1 = Event PreConstruct()
   > n2 = Sequence!(execute) -> (then_0, then_1, then_2:Exec, then_3:Exec)
   > Branch(Condition=@n1.IsDesignTime)
   > PresetForDesignTime()
   UnregisterAllModularBlocks(execute=@n2.then_1, self:UIKit Modular Button Object Reference)
   > Set Language Direction() {is=title K2Node_CallFunction}
   AddTextBlock(execute=@n2.then_2) {is=title K2Node_CallFunction}
   > UpdateText()
   > AddSecondTextBlock() {is=title K2Node_CallFunction}
   > UpdateSecondText()
   > n3 = Branch(Condition=Get UseIconBlock() -> (UseIconBlock:bool))
   > AddIconBlock() {is=title K2Node_CallFunction}
   > n4 = UpdateIcon()
   AddBackground(execute=[@n4.then, @n5.then]) {is=title K2Node_CallFunction}
   > UpdateBackground()
   > AddStatusIndicator() {is=title K2Node_CallFunction}
   > UpdateStatusIndicator()
   > UpdateInputPrompt()
   > UpdateButton()
   n5 = RemoveIconBlock(execute=@n3.else)
   SetBlocksTiming!(
    execute=@n2.then_3
    self:UIKit Modular Button Object Reference
    Timing:"UIKit Block Timing Structure (by ref)"=Make UIKit Block Timing!(
     Hovering:float=0
     Unhovering:float=0
     Focusing:float=0.3
     Unfocusing:float=0.15
     Pressing:float=0.1
     Releasing:float=0.1
     Disabling:float=0.1
     Enabling:float=0.1
     Selecting:float=0.2
     DeselectingIdle:float=0.2
     DeselectingFocused:float=0.2
    ) -> (UIKitBlockTiming:UIKit Block Timing Structure) {is=title K2Node_MakeStruct}
   ) -> (then)
   > ResetButtonState()
   Event Construct()
   > SetPrompt()
   n6 = Event OnActionProgress!() -> (OutputDelegate, then, HeldPercent:float) {is=title K2Node_Event}
   > SetHoldPercentage(HoldPercentage=@n6.HeldPercent)
   > Call On Hold Action Progress!(execute:Exec, self:WBP UIKit Button Generic Base Object Reference, HoldPercentage:double=@n6.HeldPercent) -> (then:Exec) {is=title K2Node_CallDelegate}
   Event On Deselected()
   > SetBlocksIsSelected!(execute, self, IsSelected:bool=false) -> (then)
   n7 = IsHovered()
   n8 = Get EnableFocusedSelectedState() -> (EnableFocusedSelectedState:bool)
   > RequestTransitionOnBlocks!(
    execute
    self:UIKit Modular Button Object Reference
    State:EUIKitBlockVisualState Enum=Select!(
     Option 0:EUIKitBlockVisualState Enum=DeselectedIdle
     Option 1:EUIKitBlockVisualState Enum=DeselectedFocused
     Index:bool=AND Boolean(A=@n8, B=@n7)
    ) -> (ReturnValue:EUIKitBlockVisualState Enum)
   ) -> (then)
   Event On Clicked()
   > Branch(Condition=IsPromptHoldable!() -> (Holdable:bool))
   > SetHoldPercentage(HoldPercentage=0)
   > ResetButtonState()
   > PlaySoundHoldCompleted()
   > Call On Hold Completed!(execute:Exec, self:WBP UIKit Button Generic Base Object Reference) -> (then:Exec) {is=title K2Node_CallDelegate}
   Event On Input Action Triggered() {is=title K2Node_Event}
   > PlaySoundPressed()
   Event Destruct()
   > FadeOutSound()
   n9 = Event On Locked Changed!() -> (OutputDelegate, then, bIsLocked:bool) {is=title K2Node_Event}
   > Branch(Condition=Get UseDisabledVisualsOnLocked() -> (UseDisabledVisualsOnLocked:bool))
   > n10 = Set LastEnabledEvent(LastEnabledEvent:bool=NOT Boolean(A=@n9.bIsLocked)) -> (Output_Get:bool)
   DelayUntilNextTick!(execute=[@n14.then, @n13.then, @n10.then]) -> (then)
   > SetButtonIsEnabled()
   Event OnInitialized()
   > FORT-833205 RetrieveMinDimensions()
   Event On Hovered()
   > ShowPromptFallbackAction!(execute, self, Show:bool=true) -> (then)
   > RequestTransitionOnBlocks(State=Focused)
   > PlaySoundHovered()
   Event On Unhovered()
   > ShowPromptFallbackAction(Show=false)
   > RequestTransitionOnBlocks(State=Unfocused)
   Event On Pressed() {is=title K2Node_Event}
   > RequestTransitionOnBlocks(State=Pressed)
   > n11 = Branch(Condition=IsPromptHoldable())
   > PlaySoundHoldPressed()
   PlaySoundPressed(execute=@n11.else)
   Event On Released() {is=title K2Node_Event}
   > Call On Released!(execute:Exec, self:WBP UIKit Button Generic Base Object Reference, Button:WBP UIKit Button Generic Base Object Reference=Self-Reference()) -> (then:Exec) {is=title K2Node_CallDelegate}
   > RequestTransitionOnBlocks(State=Released)
   n12 = Get CurrentHoldProgress() -> (CurrentHoldProgress:double)
   > Branch(Condition=AND Boolean(A=IsPromptHoldable(), B="float < float"(A=@n12, B=1)))
   > PlaySoundHoldReleased()
   Event On Disabled() {is=title K2Node_Event}
   > n13 = Set LastEnabledEvent(LastEnabledEvent=false)
   Event On Enabled() {is=title K2Node_Event}
   > n14 = Set LastEnabledEvent(LastEnabledEvent=true)
   Event On Selected()
   > SetBlocksIsSelected(IsSelected=true)
   > n15 = Branch(Condition=Get IsInstantSelection() -> (IsInstantSelection:bool))
   > RequestInstantTransitionOnBlocks!(execute, self:UIKit Modular Button Object Reference, State:EUIKitBlockInstantTransitionState Enum=InstantSelected) -> (then)
   > Set IsInstantSelection(IsInstantSelection:bool=false) -> (Output_Get:bool)
   n16 = IsHovered()
   n17 = Get EnableFocusedSelectedState()
   RequestTransitionOnBlocks(execute=@n15.else, State=Select(Option 0=Selected, Option 1=Focused, Index=AND Boolean(A=@n17, B=@n16)))
   n18 = Event On Input Method Changed!() -> (OutputDelegate, then, CurrentInputType:ECommonInputType Enum) {is=title K2Node_Event}
   > n19 = Set CurrentInputType(CurrentInputType:ECommonInputType Enum=@n18.CurrentInputType) -> (Output_Get:ECommonInputType Enum)
   > UpdatePrompt!(execute, self, InputType:ECommonInputType Enum=@n19) -> (then)
   > FadeOutSound()
   Event On Widget Styled() {is=title K2Node_Event}
  ]}
 FORT-833205 RetrieveMinDimensions {nodes [
   FORT-833205 RetrieveMinDimensions() {is=title K2Node_FunctionEntry}
   n1 = Get MinHeightOverride() -> (MinHeightOverride:int=0)
   n2 = Get MinWidthOverride() -> (MinWidthOverride:int=0)
   > Branch(Condition=OR Boolean(A="integer > integer"(A=@n2), B="integer > integer"(A=@n1)))
   > SetMinDimensions!(
    execute
    self:Common Button Base Object Reference
    InMinWidth:int=Get MinWidthOverride() -> (MinWidthOverride=0)
    InMinHeight:int=Get MinHeightOverride() -> (MinHeightOverride=0)
   ) -> (then)
   n3 = Self-Reference()
   n4 = GetOuterObject!(Object:Object Reference=GetOuterObject(Object=@n3)) -> (ReturnValue:Object Reference)
   > LogString(InString=Append!(
     A="FORT - 833205 : Usage of MinWidthOverride or MinHeightOverride is deprecated. Please use MinWidth/MinHeight or the Mobileversion. "
     B=GetDisplayName(Object=@n4)
     C:String=" "
     D:String=GetDisplayName(Object=@n3)
    ) -> (ReturnValue))
   n5 = Get MobileMinHeight() -> (MobileMinHeight:int=0)
   n6 = "integer > integer"(A=@n5)
   n7 = "Equal (Integer)"(A=Get MinHeight() -> (MinHeight:int=0), B=@n5)
   n8 = Get MobileMinWidth() -> (MobileMinWidth:int=0)
   n9 = "integer > integer"(A=@n8)
   n10 = "Equal (Integer)"(A=Get MinWidth() -> (MinWidth:int=0), B=@n8)
   Branch(Condition=OR Boolean(A=AND Boolean(A=@n10, B=@n9), B=AND Boolean(A=@n7, B=@n6)))
   > PrintString!(
    execute
    InString=There might be a corrupted values as MinWidth is the same as MobileMinWidth
    bPrintToScreen
    bPrintToLog
    TextColor=(R=1,G=0,B=0,A=1)
    Duration=4
    Key
   ) -> (then)
  ]}
 FadeOutSound {nodes [
   FadeOutSound() {is=title K2Node_FunctionEntry}
   > n1 = Sequence!(execute) -> (then_0, then_1)
   > n2 = Get!(execute) -> (then, else, PressedSoundAudioComponent:Audio Component Object Reference)
   > FadeOut!(
    execute
    self:Audio Component Object Reference=@n2
    FadeOutDuration:float=Get FadeOutTime() -> (FadeOutTime:double)
    FadeVolumeLevel:float=0
    FadeCurve:EAudioFaderCurve Enum=Linear
   ) -> (then)
   n3 = Get!(execute=@n1.then_1) -> (then, else, HoldSoundAudioComponent:Audio Component Object Reference)
   > FadeOut(self=@n3, FadeOutDuration=Get FadeOutTime(), FadeVolumeLevel=0, FadeCurve=Linear)
  ]}
 GetBackgroundDynamicMaterial {nodes [
   GetBackgroundDynamicMaterial() {is=title K2Node_FunctionEntry}
   > n1 = Get!(execute) -> (then, else, Block_Background:WBP UIKit Block Background Base Object Reference)
   > Return Node!(
    execute
    Dynamic Material:Material Instance Dynamic Object Reference=Get BackgroundDynamicMaterial!(self:WBP UIKit Block Background Base Object Reference=@n1) -> (BackgroundDynamicMaterial:Material Instance Dynamic Object Reference)
   )
  ]}
 GetContentHorizontalAlignment {nodes [
   GetContentHorizontalAlignment() {is=title K2Node_FunctionEntry}
   > Return Node!(
    execute
    HorizontalAlignment:EHorizontalAlignment Enum=Get ContentHorizontalAlignment() -> (ContentHorizontalAlignment:EHorizontalAlignment Enum=HAlign_Fill)
   )
  ]}
 GetContentVerticalAlignment {nodes [
   GetContentVerticalAlignment() {is=title K2Node_FunctionEntry}
   > Return Node!(
    execute
    VerticalAlignment:EVerticalAlignment Enum=Get ContentVerticalAlignment() -> (ContentVerticalAlignment:EVerticalAlignment Enum=VAlign_Fill)
   )
  ]}
 IsPromptHoldable {nodes [
   IsPromptHoldable() {is=title K2Node_FunctionEntry}
   n1 = Get InputActionWidget() -> (InputActionWidget:Common Action Widget Object Reference)
   > Return Node!(
    execute
    Holdable:bool=OR Boolean(A=Get bRequiresHold() -> (bRequiresHold:bool), B=IsHeldAction!(self:Common Action Widget Object Reference=@n1) -> (ReturnValue:bool))
   )
  ]}
 IsPromptIcon {nodes [
   IsPromptIcon() {is=title K2Node_FunctionEntry}
   n1 = GetIcon!(self:Common Action Widget Object Reference=Get InputActionWidget()) -> (ReturnValue:Slate Brush Structure)
   > Return Node!(
    execute
    Valid:bool=IsValid(Object=Break Slate Brush!(SlateBrush:"Slate Brush Structure (by ref)"=@n1) -> (
      TintColor:Slate Color Structure
      DrawAs:ESlateBrushDrawType Enum
      Tiling:ESlateBrushTileType Enum
      Mirroring:ESlateBrushMirrorType Enum
      ImageSize:Vector 2D Structure
      Margin:Margin Structure
      ResourceObject:Object Reference
      OutlineSettings:Slate Brush Outline Settings Structure
     ) {is=title K2Node_BreakStruct}.ResourceObject)
   )
  ]}
 OnBlockTextSet {nodes [OnBlockTextSet() {is=title K2Node_FunctionEntry}]}
 OnHoldActionProgress {nodes [OnHoldActionProgress!() -> (then, HoldPercentage:double=~) {is=title K2Node_FunctionEntry}]}
 OnHoldCompleted {nodes [OnHoldCompleted() {is=title K2Node_FunctionEntry}]}
 OnReleased {nodes [OnReleased!() -> (then, Button:WBP UIKit Button Generic Base Object Reference) {is=title K2Node_FunctionEntry}]}
 PlaySoundHoldCompleted {nodes [
   PlaySoundHoldCompleted() {is=title K2Node_FunctionEntry}
   > n1 = Get!(execute) -> (then, else, HoldCompletedSound:Sound Base Object Reference)
   > SpawnSound2D!(
    execute
    Sound:Sound Base Object Reference=@n1
    VolumeMultiplier:float=1
    PitchMultiplier:float=1
    StartTime:float=0
    ConcurrencySettings:Sound Concurrency Object Reference
    bPersistAcrossLevelTransition:bool=true
    bAutoDestroy:bool=true
   ) -> (then, ReturnValue:Audio Component Object Reference)
   > FadeOutSound() {is=title K2Node_CallFunction}
  ]}
 PlaySoundHoldPressed {nodes [
   PlaySoundHoldPressed() {is=title K2Node_FunctionEntry}
   > n1 = Sequence()
   > n2 = Get!(execute) -> (then, else, HoldPressedSound:Sound Base Object Reference)
   > n3 = SpawnSound2D(Sound=@n2, VolumeMultiplier=1, PitchMultiplier=1, StartTime=0, bPersistAcrossLevelTransition=false, bAutoDestroy=true)
   > Set PressedSoundAudioComponent(PressedSoundAudioComponent:Audio Component Object Reference=@n3) -> (Output_Get:Audio Component Object Reference)
   n4 = Get!(execute=@n1.then_1) -> (then, else, HoldSound:Sound Base Object Reference)
   > n5 = SpawnSound2D(Sound=@n4, VolumeMultiplier=1, PitchMultiplier=1, StartTime=0, bPersistAcrossLevelTransition=false, bAutoDestroy=true)
   > Set HoldSoundAudioComponent(HoldSoundAudioComponent:Audio Component Object Reference=@n5) -> (Output_Get:Audio Component Object Reference)
  ]}
 PlaySoundHoldReleased {nodes [
   PlaySoundHoldReleased() {is=title K2Node_FunctionEntry}
   > n1 = Get!(execute) -> (then, else, HoldReleasedSound:Sound Base Object Reference)
   > n2 = SpawnSound2D(Sound=@n1, VolumeMultiplier=1, PitchMultiplier=1, StartTime=0, bPersistAcrossLevelTransition=false, bAutoDestroy=true)
   FadeOutSound(execute=[@n2.then, @n1.else])
  ]}
 PlaySoundHovered {nodes [
   PlaySoundHovered() {is=title K2Node_FunctionEntry}
   > n1 = Get!(execute) -> (then, else, HoveredSound:Sound Base Object Reference)
   > SpawnSound2D(Sound=@n1, VolumeMultiplier=1, PitchMultiplier=1, StartTime=0, bPersistAcrossLevelTransition=false, bAutoDestroy=true)
  ]}
 PlaySoundPressed {nodes [
   PlaySoundPressed() {is=title K2Node_FunctionEntry}
   > n1 = Get!(execute) -> (then, else, PressedSound:Sound Base Object Reference)
   > SpawnSound2D(Sound=@n1, VolumeMultiplier=1, PitchMultiplier=1, StartTime=0, bPersistAcrossLevelTransition=false, bAutoDestroy=true)
  ]}
 PresetForDesignTime {nodes [
   PresetForDesignTime() {is=title K2Node_FunctionEntry}
   > n1 = Sequence()
   n2 = GetInputAction!(self:Common Button Base Object Reference) -> (InputActionRow:Data Table Row Handle Structure, ReturnValue:bool)
   > n3 = Branch(Condition=OR Boolean(A=IsPromptHoldable() {is=title K2Node_CallFunction}, B="Not Equal (Boolean)"(A=@n2.ReturnValue) -> (ReturnValue=~)))
   ClearPrompt(execute=@n3.else) {is=title K2Node_CallFunction}
   n4 = Get Slot_Main() -> (Slot_Main:Named Slot Object Reference)
   SetVisibility(
    execute=@n1.then_1
    self=@n4
    InVisibility=Select(
     Option 0:ESlateVisibility Enum=Collapsed
     Option 1:ESlateVisibility Enum=SelfHitTestInvisible
     Index=HasAnyChildren!(self:Panel Widget Object Reference=@n4) -> (ReturnValue:bool)
    ) -> (ReturnValue:ESlateVisibility Enum)
   )
   n5 = Get Slot_Background() -> (Slot_Background:Named Slot Object Reference)
   > SetVisibility(self=@n5, InVisibility=Select(Option 0=Collapsed, Option 1=SelfHitTestInvisible, Index=HasAnyChildren(self=@n5)))
  ]}
 RemoveIconBlock {nodes [
   RemoveIconBlock() {is=title K2Node_FunctionEntry}
   > n1 = Get!(execute) -> (then, else, Block_Icon:WBP UIKit Block Image Object Reference)
   > UnregisterModularBlock!(execute, self:UIKit Modular Button Object Reference, Block:UIKit Block Object Reference=@n1) -> (then)
   > RemoveFromParent(self=@n1)
   > Set Block_Icon()
  ]}
 RequestInstantSelection {nodes [
   RequestInstantSelection() {is=title K2Node_FunctionEntry}
   > Set IsInstantSelection(IsInstantSelection=true)
  ]}
 ResetButtonState {nodes [
   ResetButtonState() {is=title K2Node_FunctionEntry}
   > StopAllAnimations(self:User Widget Object Reference)
   > RequestInstantTransitionOnBlocks(State=Select(
     Option 0:EUIKitBlockInstantTransitionState Enum=InstantDisabled
     Option 1:EUIKitBlockInstantTransitionState Enum=InstantEnabled
     Index=GetIsEnabled!(self:Widget Object Reference) -> (ReturnValue:bool)
    ) -> (ReturnValue:EUIKitBlockInstantTransitionState Enum))
  ]}
 Set Language Direction {nodes [
   Set Language Direction() {is=title K2Node_FunctionEntry}
   > Branch(Condition="Equal Exactly (String)"(A=GetCurrentCulture!() -> (ReturnValue:String), B=ar))
   > n1 = Sequence!(execute) -> (then_0, then_1, then_2:Exec, then_3:Exec, then_4:Exec, then_5:Exec)
   n2 = Get TextBlockAlignment() -> (TextBlockAlignment=NewEnumerator0)
   n3 = "Equal (Enum)"(A:E_UI_Alignment Enum=@n2, B:E_UI_Alignment Enum=NewEnumerator2)
   Branch(execute=@n1.then_1, Condition=OR Boolean(A=@n3, B="Equal (Enum)"(A=@n2, B=NewEnumerator3)))
   > Set TextBlockAlignment(TextBlockAlignment:E_UI_Alignment Enum=Select(Option 0:E_UI_Alignment Enum=NewEnumerator2, Option 1:E_UI_Alignment Enum=NewEnumerator3, Index=@n3) -> (ReturnValue:E_UI_Alignment Enum)) -> (Output_Get:E_UI_Alignment Enum=NewEnumerator0)
   n4 = Get!() -> (TextPadding_Left:float=~, TextPadding_Top:float=~, TextPadding_Right:float=~, TextPadding_Bottom:float=~)
   > Set!(
    execute
    TextPadding_Left:float=@n4.TextPadding_Right
    TextPadding_Top:float=@n4.TextPadding_Top
    TextPadding_Right:float=@n4
    TextPadding_Bottom:float=@n4.TextPadding_Bottom
   ) -> (then, Output_Get:Margin Structure)
   n5 = Get!() -> (IconPadding_Left:float=~, IconPadding_Top:float=~, IconPadding_Right:float=~, IconPadding_Bottom:float=~)
   > Set!(
    execute
    IconPadding_Left:float=@n5.IconPadding_Right
    IconPadding_Top:float=@n5.IconPadding_Top
    IconPadding_Right:float=@n5
    IconPadding_Bottom:float=@n5.IconPadding_Bottom
   ) -> (then, Output_Get)
   n6 = Get SecondTextBlockAlignment() -> (SecondTextBlockAlignment=NewEnumerator0)
   n7 = "Equal (Enum)"(A=@n6, B=NewEnumerator2)
   Branch(execute=@n1.then_5, Condition=OR Boolean(A=@n7, B="Equal (Enum)"(A=@n6, B=NewEnumerator3)))
   > Set SecondTextBlockAlignment(SecondTextBlockAlignment:E_UI_Alignment Enum=Select(Option 0=NewEnumerator2, Option 1=NewEnumerator3, Index=@n7)) -> (Output_Get:E_UI_Alignment Enum=NewEnumerator0)
   n8 = Get!() -> (SecondTextPadding_Left:float=~, SecondTextPadding_Top:float=~, SecondTextPadding_Right:float=~, SecondTextPadding_Bottom:float=~)
   > Set!(
    execute
    SecondTextPadding_Left:float=@n8.SecondTextPadding_Right
    SecondTextPadding_Top:float=@n8.SecondTextPadding_Top
    SecondTextPadding_Right:float=@n8
    SecondTextPadding_Bottom:float=@n8.SecondTextPadding_Bottom
   ) -> (then, Output_Get)
   n9 = Get IconHorizontalAlignment() -> (IconHorizontalAlignment=HAlign_Fill)
   n10 = "Equal (Enum)"(A:EHorizontalAlignment Enum=@n9, B:EHorizontalAlignment Enum=HAlign_Left)
   Branch(execute=@n1.then_2, Condition=OR Boolean(A=@n10, B="Equal (Enum)"(A=@n9, B=HAlign_Right)))
   > Set IconHorizontalAlignment(IconHorizontalAlignment:EHorizontalAlignment Enum=Select(Option 0:EHorizontalAlignment Enum=HAlign_Left, Option 1:EHorizontalAlignment Enum=HAlign_Right, Index=@n10) -> (ReturnValue:EHorizontalAlignment Enum)) -> (Output_Get:EHorizontalAlignment Enum=HAlign_Fill)
   n11 = Get PromptHorizontalAlignment() -> (PromptHorizontalAlignment:EHorizontalAlignment Enum=HAlign_Fill)
   n12 = "Equal (Enum)"(A=@n11, B=HAlign_Left)
   n13 = "Equal (Enum)"(A=@n11, B=HAlign_Right)
   Branch(
    execute=@n1.then_3
    Condition=AND Boolean(A=OR Boolean(A=@n12, B=@n13), B=Get ShouldPromptDirectionFollowCulture() -> (ShouldPromptDirectionFollowCulture:bool))
   )
   > Set PromptHorizontalAlignment(PromptHorizontalAlignment:EHorizontalAlignment Enum=Select(Option 0=HAlign_Left, Option 1=HAlign_Right, Index=@n12)) -> (Output_Get:EHorizontalAlignment Enum=HAlign_Fill)
   n14 = Get!() -> (PromptNudge_X:double, PromptNudge_Y:double)
   > Set!(execute, PromptNudge_X:double=float * float(A=@n14, B=-1) {is=title K2Node_PromotableOperator}, PromptNudge_Y:double=@n14.PromptNudge_Y) -> (then, Output_Get:Vector 2D Structure)
   n15 = Get StatusIndicatorHorizontalAlignment() -> (StatusIndicatorHorizontalAlignment:EHorizontalAlignment Enum=HAlign_Fill)
   n16 = "Equal (Enum)"(A=@n15, B=HAlign_Left)
   Branch(execute=@n1.then_4, Condition=OR Boolean(A=@n16, B="Equal (Enum)"(A=@n15, B=HAlign_Right)))
   > Set StatusIndicatorHorizontalAlignment(StatusIndicatorHorizontalAlignment:EHorizontalAlignment Enum=Select(Option 0=HAlign_Left, Option 1=HAlign_Right, Index=@n16)) -> (Output_Get:EHorizontalAlignment Enum=HAlign_Fill)
   n17 = Get!() -> (
    StatusIndicatorPadding_Left:float=~
    StatusIndicatorPadding_Top:float=~
    StatusIndicatorPadding_Right:float=~
    StatusIndicatorPadding_Bottom:float=~
   )
   > Set!(
    execute
    StatusIndicatorPadding_Left:float=@n17.StatusIndicatorPadding_Right
    StatusIndicatorPadding_Top:float=@n17.StatusIndicatorPadding_Top
    StatusIndicatorPadding_Right:float=@n17
    StatusIndicatorPadding_Bottom:float=@n17.StatusIndicatorPadding_Bottom
   ) -> (then, Output_Get:Margin Structure)
   n18 = Get ContentHorizontalAlignment() -> (ContentHorizontalAlignment=HAlign_Fill)
   n19 = "Equal (Enum)"(A=@n18, B=HAlign_Left)
   Branch(execute=@n1.then_0, Condition=OR Boolean(A=@n19, B="Equal (Enum)"(A=@n18, B=HAlign_Right)))
   > Set ContentHorizontalAlignment(ContentHorizontalAlignment:EHorizontalAlignment Enum=Select(Option 0=HAlign_Left, Option 1=HAlign_Right, Index=@n19)) -> (Output_Get:EHorizontalAlignment Enum=HAlign_Fill)
   n20 = Get!() -> (ButtonPadding_Left:float=~, ButtonPadding_Top:float=~, ButtonPadding_Right:float=~, ButtonPadding_Bottom:float=~)
   > Set!(
    execute
    ButtonPadding_Left:float=@n20.ButtonPadding_Right
    ButtonPadding_Top:float=@n20.ButtonPadding_Top
    ButtonPadding_Right:float=@n20
    ButtonPadding_Bottom:float=@n20.ButtonPadding_Bottom
   ) -> (then, Output_Get)
  ]}
 SetBackgroundMaterials {nodes [
   n1 = SetBackgroundMaterials!() -> (then, Background MI:Material Instance Object Reference, Outline MI:Material Instance Object Reference) {is=title K2Node_FunctionEntry}
   > n2 = Set BackgroundMaterial(BackgroundMaterial:Material Instance Object Reference=@n1) -> (Output_Get:Material Instance Object Reference)
   > n3 = Set OutlineMaterial(OutlineMaterial:Material Instance Object Reference=@n1.Outline MI) -> (Output_Get:Material Instance Object Reference)
   > n4 = Get!(execute:Exec) -> (then:Exec, else:Exec, Block_Background:WBP UIKit Block Background Base Object Reference)
   > SetBackgroundMaterial!(execute, self:WBP UIKit Block Background Base Object Reference=@n4, Material Instance:"Material Instance Object Reference (by ref)"=@n2) -> (then)
   > SetOutlineMaterial!(execute, self:WBP UIKit Block Background Base Object Reference=@n4, Material Instance:"Material Instance Object Reference (by ref)"=@n3) -> (then)
  ]}
 SetBlocksIsSelected {nodes [
   n1 = SetBlocksIsSelected!() -> (then, IsSelected:bool=~) {is=title K2Node_FunctionEntry}
   > n2 = For Each Loop(Array:Array of UIKit Block Object References=Get ModularBlocks() -> (ModularBlocks:Array of UIKit Block Object References)) -> (Array Element:UIKit Block Object Reference)
   > n3 = Cast To WBP_UIKit_Block_Base!(execute, Object=@n2) -> (then, CastFailed, AsWBP UIKit Block Base:WBP UIKit Block Base Object Reference) {is=title K2Node_DynamicCast}
   > SetIsSelected!(execute, self:WBP UIKit Block Base Object Reference=@n3, IsSelected:bool=@n1) -> (then)
  ]}
 SetButtonContentAlignment {nodes [
   n1 = SetButtonContentAlignment!() -> (then, HorizontalAlignment:EHorizontalAlignment Enum=HAlign_Fill, VerticalAlignment:EVerticalAlignment Enum=VAlign_Fill) {is=title K2Node_FunctionEntry}
   > n2 = Set ContentHorizontalAlignment(ContentHorizontalAlignment=@n1) -> (Output_Get=HAlign_Fill)
   > n3 = Set ContentVerticalAlignment(ContentVerticalAlignment:EVerticalAlignment Enum=@n1.VerticalAlignment) -> (Output_Get:EVerticalAlignment Enum=VAlign_Fill)
   > SetGridSlotAlignment(Widget=Get Main(), Horizontal=@n2, Vertical=@n3)
  ]}
 SetButtonIsEnabled {nodes [
   SetButtonIsEnabled() {is=title K2Node_FunctionEntry}
   > n1 = Branch(Condition=Get LastEnabledEvent() -> (LastEnabledEvent:bool))
   > n2 = Branch(Condition=IsHovered())
   > n3 = RequestTransitionOnBlocks(State=Enabled)
   SetPrompt(execute=[@n3.then, @n4.then])
   n4 = RequestInstantTransitionOnBlocks(execute=@n2.else, State=InstantEnabled)
   RequestTransitionOnBlocks(execute=@n1.else, State=Disabled)
   > SetVisibility(self=Get Prompt(), InVisibility=Collapsed)
  ]}
 SetButtonMinSize {nodes [
   n1 = SetButtonMinSize!() -> (then, Min Width:int, Min Height:int) {is=title K2Node_FunctionEntry}
   > SetMinDimensions(InMinWidth=@n1, InMinHeight=@n1.Min Height)
  ]}
 SetButtonPadding {nodes [
   n1 = SetButtonPadding!() -> (then, Padding:Margin Structure) {is=title K2Node_FunctionEntry}
   > n2 = Set ButtonPadding(ButtonPadding:Margin Structure=@n1) -> (Output_Get:Margin Structure)
   > SetGridSlotPadding(Widget=Get Main(), Padding=@n2)
  ]}
 SetData {nodes [SetData!() -> (then, VM Button:UI Kit Selectable Button View Model Object Reference) {is=title K2Node_FunctionEntry}]}
 SetFont {nodes [
   n1 = SetFont!() -> (then, In Font:Slate Font Info Structure) {is=title K2Node_FunctionEntry}
   > n2 = Set Font(Font:Slate Font Info Structure=@n1) -> (Output_Get:Slate Font Info Structure)
   > n3 = Get!(execute) -> (then, else, Block_Text:WBP UIKit Block Text Object Reference)
   > SetFont!(execute, self:WBP UIKit Block Text Object Reference=@n3, Font:"Slate Font Info Structure (by ref)"=@n2) -> (then) {is=title K2Node_CallFunction}
   > SetMobileFontSizeMultiplier!(
    execute
    self:WBP UIKit Block Text Object Reference=@n3
    MobileFontSizeMultiplier:double=Select(
     Option 0:double=Get MobileFontSizeMultiplier() -> (MobileFontSizeMultiplier:double)
     Option 1:double=Get MobileFontSizeMultiplierOverride() -> (MobileFontSizeMultiplierOverride:double)
     Index=Get OverrideDefaultMobileFontSizeMultiplier() -> (OverrideDefaultMobileFontSizeMultiplier:bool)
    ) -> (ReturnValue:double)
   ) -> (then)
  ]}
 SetHoldPercentage {nodes [
   n1 = SetHoldPercentage!() -> (then, HoldPercentage:double=~) {is=title K2Node_FunctionEntry}
   > n2 = Set CurrentHoldProgress(CurrentHoldProgress:double=@n1) -> (Output_Get:double)
   > SetScalarParameterValue(
    self=Get InputActionProgressMaterial() -> (InputActionProgressMaterial:Material Instance Dynamic Object Reference)
    ParameterName=Percentage
    Value=@n2
   )
  ]}
 SetIcon {nodes [
   n1 = SetIcon!() -> (then, Icon:Object Reference) {is=title K2Node_FunctionEntry}
   > n2 = Branch(Condition=IsValid(Object=@n1))
   > n3 = Cast To MaterialInstance!(execute, Object=@n1) -> (then, CastFailed, AsMaterial Instance:Material Instance Object Reference) {is=title K2Node_DynamicCast}
   > SetIconMaterial!(execute, self, Material Instance:Material Instance Object Reference=@n3) -> (then)
   n4 = Cast To Texture2D(execute=@n3.CastFailed, Object=@n1)
   > SetIconTexture!(execute, self, Icon:Texture 2D Object Reference=@n4) -> (then)
   n5 = Self-Reference()
   PrintString(
    execute=@n4.CastFailed
    InString=Append!(A="WARNING: ", B=GetDisplayName(Object=@n5), C=" - Incorrect Icon type in SetIcon fucntion.") -> (ReturnValue)
    TextColor=(R=1,G=0,B=0.5683,A=1)
    Duration=8
   )
   RemoveIconBlock(execute=@n2.else) {is=title K2Node_CallFunction}
   > Set Icon(Icon:Texture 2D Object Reference) -> (Output_Get:Texture 2D Object Reference)
  ]}
 SetIconAlignment {nodes [
   n1 = SetIconAlignment!() -> (then, Horizontal:EHorizontalAlignment Enum=HAlign_Fill, Vertical:EVerticalAlignment Enum=VAlign_Fill) {is=title K2Node_FunctionEntry}
   > n2 = Set IconHorizontalAlignment(IconHorizontalAlignment=@n1) -> (Output_Get=HAlign_Fill)
   > n3 = Set IconVerticalAlignment(IconVerticalAlignment:EVerticalAlignment Enum=@n1.Vertical) -> (Output_Get:EVerticalAlignment Enum=VAlign_Fill)
   > n4 = Get!(execute) -> (then, else, Block_Icon:WBP UIKit Block Image Object Reference)
   > SetGridSlotAlignment(Widget=@n4, Horizontal=@n2, Vertical=@n3)
  ]}
 SetIconMaterial {nodes [
   n1 = SetIconMaterial!() -> (then, Material Instance:Material Instance Object Reference) {is=title K2Node_FunctionEntry}
   > n2 = Set IconMaterial(IconMaterial:Material Instance Object Reference=@n1) -> (Output_Get:Material Instance Object Reference)
   > n3 = Branch(Condition=IsValid(Object=@n2))
   > n4 = Get()
   n5 = Get Block_Icon()
   SetMaterial!(execute=[@n6.then, @n4.then], self:WBP UIKit Block Image Object Reference=@n5, InMaterial:"Material Instance Object Reference (by ref)"=@n2) -> (then)
   > RequestInstantTransition!(
    execute
    self:UIKit Block Object Reference=@n5
    State:EUIKitBlockInstantTransitionState Enum=Select(
     Option 0:EUIKitBlockInstantTransitionState Enum=InstantEnabled
     Option 1:EUIKitBlockInstantTransitionState Enum=InstantSelected
     Index=GetSelected!(self:Common Button Base Object Reference) -> (ReturnValue:bool)
    ) -> (ReturnValue:EUIKitBlockInstantTransitionState Enum)
   ) -> (then)
   n6 = AddIconBlock(execute=@n4.else)
   RemoveIconBlock(execute=@n3.else)
  ]}
 SetIconPadding {nodes [
   n1 = SetIconPadding!() -> (then, Padding:Margin Structure) {is=title K2Node_FunctionEntry}
   > n2 = Set IconPadding(IconPadding:Margin Structure=@n1) -> (Output_Get:Margin Structure)
   > n3 = Get()
   > Set Padding!(execute, Padding:Margin Structure=@n2, self:User Widget Object Reference=@n3) -> (then, Output_Get:Margin Structure)
  ]}
 SetIconSize {nodes [
   n1 = SetIconSize!() -> (then, Icon:Vector 2D Structure) {is=title K2Node_FunctionEntry}
   > n2 = Set IconSize(IconSize:Vector 2D Structure=@n1) -> (Output_Get:Vector 2D Structure)
   > n3 = Get()
   > SetSize(self=@n3, Desired Size=@n2)
  ]}
 SetIconTexture {nodes [
   n1 = SetIconTexture!() -> (then, Icon:Texture 2D Object Reference) {is=title K2Node_FunctionEntry}
   > n2 = Set Icon(Icon=@n1)
   > n3 = Branch(Condition=IsValid(Object=@n2))
   > n4 = Get()
   n5 = Get!(execute=[@n7.then, @n4.then]) -> (then, else, IconMaterial:Material Instance Object Reference)
   n6 = Get Block_Icon()
   > SetTextureWithParameter!(execute, self:WBP UIKit Block Image Object Reference=@n6, ParameterName:Name=Icon, InTexture:"Texture 2D Object Reference (by ref)"=@n2) -> (then)
   > RequestInstantTransition(self=@n6, State=Select(Option 0=InstantEnabled, Option 1=InstantSelected, Index=GetSelected()))
   SetBrush(
    execute=@n5.else
    self:WBP UIKit Block Image Object Reference=Get Block_Icon()
    InBrush=Make Slate Brush!(ResourceObject:Object Reference=@n2) -> (SlateBrush:Slate Brush Structure) {is=title K2Node_MakeStruct}
   )
   AddIconBlock(execute=@n4.else)
   > n7 = SetMaterial(self=Get Block_Icon(), InMaterial=Get IconMaterial() -> (IconMaterial:Material Instance Object Reference))
   RemoveIconBlock(execute=@n3.else)
  ]}
 SetInputIconDarkMode {nodes [
   n1 = SetInputIconDarkMode!() -> (then, UseDark:bool=~) {is=title K2Node_FunctionEntry}
   > Set UseDarkInputIcon(UseDarkInputIcon:bool=@n1) -> (Output_Get:bool)
   > n2 = GetIconDynamicMaterial!(execute, self:Common Action Widget Object Reference=Get InputActionWidget()) -> (then, ReturnValue:Material Instance Dynamic Object Reference)
   > n3 = Set CurrentDynamicPromptIcon(CurrentDynamicPromptIcon:Material Instance Dynamic Object Reference=@n2) -> (Output_Get:Material Instance Dynamic Object Reference)
   > SetScalarParameterValue(self=@n3, ParameterName=Use Dark, Value=Select(Option 0:float=0, Option 1:float=1, Index=Get UseDarkInputIcon()) -> (ReturnValue:float))
   > SetScalarParameterValue(
    self=Get CurrentDynamicPromptIcon() -> (CurrentDynamicPromptIcon:Material Instance Dynamic Object Reference)
    ParameterName=Outline Opacity
    Value=Select(Option 0=0.5, Option 1=0, Index=Get UseDarkInputIcon())
   )
  ]}
 SetPreviewData {nodes [SetPreviewData!() -> (then, VM:MVVM Base Viewmodel Object Reference) {is=title K2Node_FunctionEntry}]}
 SetPrompt {nodes [
   SetPrompt() {is=title K2Node_FunctionEntry}
   n1 = Get CommonInputSubsystem!() -> (ReturnValue:Common Input Object Reference) {is=title K2Node_GetBlueprintContextEx}
   > Branch(Condition=IsValid(Object=@n1))
   > n2 = GetDynamicMaterial(self=Get HoldProgressBar())
   > Set InputActionProgressMaterial(InputActionProgressMaterial:Material Instance Dynamic Object Reference=@n2) -> (Output_Get:Material Instance Dynamic Object Reference)
   > n3 = Set CurrentInputType(CurrentInputType=GetCurrentInputType(self=@n1))
   > UpdatePrompt(InputType=@n3)
  ]}
 SetSecondText {nodes [
   n1 = SetSecondText!() -> (then, SecondText:Text) {is=title K2Node_FunctionEntry}
   > n2 = Set SecondText(SecondText:Text=@n1) -> (Output_Get:Text)
   > n3 = Get!(execute) -> (then, else, Block_SecondText:WBP UIKit Block Text Object Reference)
   > SetText!(execute, self:WBP UIKit Block Text Object Reference=@n3, Text:Text=@n2) -> (then)
  ]}
 SetStatusIndicator {nodes [
   n1 = SetStatusIndicator!() -> (then, BangType:EFortBangType Enum=Invalid) {is=title K2Node_FunctionEntry}
   > n2 = Get!(execute) -> (then, else, StatusIndicator:WBP UIKit Status Indicator Object Reference)
   > PassBangTypeFromBP!(execute, self:WBP UIKit Status Indicator Base Object Reference=@n2, EBangType:EFortBangType Enum=@n1) -> (then)
  ]}
 SetStatusIndicatorConfiguration {nodes [
   n1 = SetStatusIndicatorConfiguration!() -> (then, Configuration:E_UI_StatusIndicator_Configuration Enum=NewEnumerator0, Count:int) {is=title K2Node_FunctionEntry}
   > n2 = Set Configuration(Configuration:E_UI_StatusIndicator_Configuration Enum=@n1) -> (Output_Get:E_UI_StatusIndicator_Configuration Enum=NewEnumerator0)
   > n3 = Get()
   > SetIndicatorConfiguration!(
    execute
    self:WBP UIKit Status Indicator Base Object Reference=@n3
    InCount:int=@n1.Count
    IndicatorConfig:E_UI_StatusIndicator_Configuration Enum=@n2
   ) -> (then)
  ]}
 SetStatusIndicatorType {nodes [
   n1 = SetStatusIndicatorType!() -> (then, Type:E_UI_StatusIndicator_Type Enum=NewEnumerator0) {is=title K2Node_FunctionEntry}
   > n2 = Set Type(Type:E_UI_StatusIndicator_Type Enum=@n1) -> (Output_Get:E_UI_StatusIndicator_Type Enum=NewEnumerator0)
   > n3 = Get()
   > SetIndicatorType!(execute, self:WBP UIKit Status Indicator Base Object Reference=@n3, IndicatorType:E_UI_StatusIndicator_Type Enum=@n2) -> (then)
  ]}
 SetText {nodes [
   n1 = SetText!() -> (then, Text:Text) {is=title K2Node_FunctionEntry}
   > n2 = Set Text!(execute, Text=@n1) -> (then, Output_Get)
   > n3 = Get!(execute) -> (then, else, Block_Text:WBP UIKit Block Text Object Reference)
   > SetText(self=@n3, Text=@n2) {is=title K2Node_CallFunction}
  ]}
 SetTextNudge {nodes [
   n1 = SetTextNudge!() -> (then, TextNudge:Vector 2D Structure) {is=title K2Node_FunctionEntry}
   > Set TextNudge(TextNudge:Vector 2D Structure=@n1) -> (Output_Get:Vector 2D Structure)
   > n2 = Get()
   > n3 = Branch(Condition=Get ShouldTextImpactLayout())
   > n4 = Branch(Condition=Get IsTextOutsideBackground())
   > SetGridSlotNudge!(execute, Widget:Widget Object Reference=@n2, Nudge:Vector 2D Structure=Get TextNudge() -> (TextNudge:Vector 2D Structure)) -> (then)
   Reroute Node(InputPin=@n4.else)
   SetGridSlotNudge(execute=@n3.else, Widget=Get VisualAttachment(), Nudge=Get TextNudge())
  ]}
 SetTextOverflowPolicy {nodes [
   n1 = SetTextOverflowPolicy!() -> (then, OverflowPolicy:ETextOverflowPolicy Enum=Clip) {is=title K2Node_FunctionEntry}
   > n2 = Get()
   > SetTextOverflowPolicy!(execute, self:WBP UIKit Block Text Object Reference=@n2, OverflowPolicy:ETextOverflowPolicy Enum=@n1) -> (then) {is=title K2Node_CallFunction}
  ]}
 SetTextPadding {nodes [
   n1 = SetTextPadding!() -> (then, TextPadding:Margin Structure) {is=title K2Node_FunctionEntry}
   > Set TextPadding(TextPadding:Margin Structure=@n1) -> (Output_Get:Margin Structure)
   > n2 = Get()
   > n3 = Branch(Condition=Get ShouldTextImpactLayout())
   > n4 = Branch(Condition=Get IsTextOutsideBackground())
   > SetGridSlotPadding(Widget=@n2, Padding=Get TextPadding() -> (TextPadding:Margin Structure))
   Set Padding(execute=@n4.else, Padding=Get TextPadding(), self=@n2)
   SetGridSlotPadding(execute=@n3.else, Widget=Get VisualAttachment(), Padding=Get TextPadding())
  ]}
 ShowPromptFallbackAction {nodes [
   n1 = ShowPromptFallbackAction!() -> (then, Show:bool=~) {is=title K2Node_FunctionEntry}
   n2 = Get CurrentInputType() -> (CurrentInputType:ECommonInputType Enum)
   > Branch(Condition=AND Boolean(
     A="Equal (Enum)"(A:ECommonInputType Enum=@n2, B:ECommonInputType Enum=Gamepad)
     B=Get bShouldUseFallbackDefaultInputAction() -> (bShouldUseFallbackDefaultInputAction:bool)
    ))
   > SetVisibility(
    self=Get Prompt()
    InVisibility=Select(Option 0:ESlateVisibility Enum=Collapsed, Option 1:ESlateVisibility Enum=SelfHitTestInvisible, Index=@n1) -> (ReturnValue:ESlateVisibility Enum)
   )
  ]}
 UpdateBackground {nodes [
   UpdateBackground() {is=title K2Node_FunctionEntry}
   > n1 = Get!(execute) -> (then, else, Block_Background:WBP UIKit Block Background Base Object Reference)
   > n2 = Sequence!(execute) -> (then_0, then_1)
   n3 = Get BackgroundMaterial() -> (BackgroundMaterial:Material Instance Object Reference)
   > Branch(Condition=IsValid(Object=@n3))
   > SetBackgroundMaterial(self=@n1, Material Instance=@n3)
   n4 = Get OutlineMaterial() -> (OutlineMaterial:Material Instance Object Reference)
   Branch(execute=@n2.then_1, Condition=IsValid(Object=@n4))
   > SetOutlineMaterial(self=@n1, Material Instance=@n4)
  ]}
 UpdateButton {nodes [
   UpdateButton() {is=title K2Node_FunctionEntry}
   > n1 = Branch(Condition=IsMobileGame())
   > SetMinDimensions(InMinWidth=Get MobileMinWidth() -> (MobileMinWidth=0), InMinHeight=Get MobileMinHeight() -> (MobileMinHeight=0))
   > n2 = SetGridSlotPadding(Widget=Get Main(), Padding=Get MobileButtonPadding() -> (MobileButtonPadding:Margin Structure))
   SetButtonContentAlignment!(
    execute=[@n2.then, @n10.then]
    self
    HorizontalAlignment:EHorizontalAlignment Enum=Get ContentHorizontalAlignment() -> (ContentHorizontalAlignment=HAlign_Fill)
    VerticalAlignment:EVerticalAlignment Enum=Get ContentVerticalAlignment() -> (ContentVerticalAlignment=VAlign_Fill)
   ) -> (then) {is=title K2Node_CallFunction}
   n3 = AND Boolean(A=Get UseTextBlock(), B=Get IsTextOutsideBackground())
   n4 = Get TextBlockAlignment() -> (TextBlockAlignment=NewEnumerator0)
   n5 = SelectInt(
    A=Select!(NewEnumerator0:int=1, NewEnumerator1:int=1, NewEnumerator2:int=2, NewEnumerator3:int=0, Index:E_UI_Alignment Enum=@n4) -> (ReturnValue:int)
    B=1
    bPickA=@n3
   )
   n6 = SelectInt(A=Select(NewEnumerator0=2, NewEnumerator1=0, NewEnumerator2=1, NewEnumerator3=1, Index=@n4), B=1, bPickA=@n3)
   > SetGridSlotColumnAndRow!(execute, Widget:Widget Object Reference=Get Main(), Row:int=@n6, Row Span:int=1, Column:int=@n5, Column Span:int=1) -> (then)
   > SetGridSlotColumnAndRow(Widget=Get Slot_Main(), Row=@n6, Row Span=1, Column=@n5, Column Span=1)
   n7 = Get VisualAttachment()
   > n8 = Is Valid(InputObject=@n7)
   > Set ContentAnchor!(
    execute
    ContentAnchor:Vector 2D Structure=Get VisualAttachmentContentAnchor() -> (VisualAttachmentContentAnchor:Vector 2D Structure)
    self:Common Visual Attachment Object Reference=@n7
   ) -> (then, Output_Get:Vector 2D Structure)
   > n9 = SetGridSlotAlignment(
    Widget=@n7
    Horizontal=Get VisualAttachmentHorizontalAlignment() -> (VisualAttachmentHorizontalAlignment:EHorizontalAlignment Enum=HAlign_Fill)
    Vertical=Get VisualAttachmentVerticalAlignment() -> (VisualAttachmentVerticalAlignment:EVerticalAlignment Enum=VAlign_Fill)
   )
   SetVerticalAlignment(
    execute=[@n9.then, @n8.Is Not Valid]
    self:Size Box Slot Object Reference=SlotAsSizeBoxSlot!(Widget:Widget Object Reference=Get Main_Stack()) -> (ReturnValue:Size Box Slot Object Reference)
    InVerticalAlignment=VAlign_Center
   )
   SetMinDimensions(execute=@n1.else, InMinWidth=Get MinWidth() -> (MinWidth=0), InMinHeight=Get MinHeight() -> (MinHeight=0))
   > n10 = SetGridSlotPadding(Widget=Get Main(), Padding=Get ButtonPadding() -> (ButtonPadding:Margin Structure))
  ]}
 UpdateIcon {nodes [
   UpdateIcon() {is=title K2Node_FunctionEntry}
   > n1 = Get!(execute) -> (then, else, Block_Icon:WBP UIKit Block Image Object Reference)
   > SetMaterial(self=@n1, InMaterial=Get IconMaterial())
   > SetTextureWithParameter(self=@n1, ParameterName=Icon, InTexture=Get Icon() -> (Icon:Texture 2D Object Reference))
   n2 = Get Icon()
   n3 = Get IconMaterial()
   > SetBlockVisibility!(execute, self:WBP UIKit Block Base Object Reference=@n1, Visible:bool=OR Boolean(A=IsValid(Object=@n3), B=IsValid(Object=@n2))) -> (then)
   > RequestInstantTransition(
    self=@n1
    State=Select!(
     Option 0:EUIKitBlockInstantTransitionState Enum=InstantEnabled
     Option 1:EUIKitBlockInstantTransitionState Enum=InstantSelected
     Index:bool=GetSelected()
    ) -> (ReturnValue:EUIKitBlockInstantTransitionState Enum)
   )
  ]}
 UpdateInputPrompt {nodes [
   UpdateInputPrompt() {is=title K2Node_FunctionEntry}
   n1 = Get Prompt()
   > SetGridSlotPadding(Widget=@n1, Padding=Get PromptPadding() -> (PromptPadding:Margin Structure))
   > SetGridSlotNudge(Widget=@n1, Nudge=Get PromptNudge() -> (PromptNudge:Vector 2D Structure))
   > SetGridSlotAlignment(
    Widget=@n1
    Horizontal=Get PromptHorizontalAlignment() -> (PromptHorizontalAlignment=HAlign_Fill)
    Vertical=Get PromptVerticalAlignment() -> (PromptVerticalAlignment:EVerticalAlignment Enum=VAlign_Fill)
   )
   > n2 = Branch(Condition=Get ShouldPromptImpactLayout() -> (ShouldPromptImpactLayout:bool))
   n3 = Get Prompt()
   > SetGridSlotColumnAndRow(
    Widget=@n3
    Row=Select!(
     VAlign_Fill:int=1
     VAlign_Top:int=0
     VAlign_Center:int=1
     VAlign_Bottom:int=2
     Index:EVerticalAlignment Enum=Get PromptVerticalAlignment() -> (PromptVerticalAlignment=VAlign_Fill)
    ) -> (ReturnValue:int)
    Row Span=1
    Column=Select!(
     HAlign_Fill:int=1
     HAlign_Left:int=0
     HAlign_Center:int=1
     HAlign_Right:int=2
     Index:EHorizontalAlignment Enum=Get PromptHorizontalAlignment() -> (PromptHorizontalAlignment=HAlign_Fill)
    ) -> (ReturnValue)
    Column Span=1
   )
   SetGridSlotColumnAndRow(execute=@n2.else, Widget=@n3, Row=1, Row Span=1, Column=1, Column Span=1)
  ]}
 UpdatePrompt {nodes [
   n1 = UpdatePrompt!() -> (then, InputType:ECommonInputType Enum) {is=title K2Node_FunctionEntry}
   > ClearPrompt()
   > Set IsPromptHoldable(IsPromptHoldable:bool=IsPromptHoldable()) -> (Output_Get:bool)
   > Set IsPromptIcon(IsPromptIcon:bool=IsPromptIcon!() -> (Valid:bool) {is=title K2Node_CallFunction}) -> (Output_Get:bool)
   n2 = Get bHideInputActionWithKeyboard() -> (bHideInputActionWithKeyboard:bool)
   n3 = Get IsPromptHoldable() -> (IsPromptHoldable:bool)
   n4 = Get IsPromptIcon() -> (IsPromptIcon:bool)
   n5 = Select!(
    Option 0:ESlateVisibility Enum=Collapsed
    Option 1:ESlateVisibility Enum=SelfHitTestInvisible
    Index:bool=AND Boolean(A=OR Boolean(A=@n4, B=@n3), B=NOT Boolean(A=@n2))
   ) -> (ReturnValue:ESlateVisibility Enum)
   n6 = Select(Option 0=Collapsed, Option 1=SelfHitTestInvisible, Index=Get IsPromptHoldable())
   n7 = Select(Option 0=Collapsed, Option 1=SelfHitTestInvisible, Index=Get IsPromptIcon())
   > SetVisibility(
    self=Get Prompt()
    InVisibility=Select(
     Option 0=Collapsed
     Option 1=Select!(
      MouseAndKeyboard:ESlateVisibility Enum=@n5
      Gamepad:ESlateVisibility Enum=@n7
      Touch:ESlateVisibility Enum=@n6
      Count:ESlateVisibility Enum=Collapsed
      Index:ECommonInputType Enum=@n1
     ) -> (ReturnValue)
     Index=GetIsEnabled()
    )
   )
   > Branch(Condition=Get IsPromptHoldable())
   n8 = Get IsPromptIcon()
   > SetScalarParameterValue(
    self=Get InputActionProgressMaterial()
    ParameterName=ShapeSwitch
    Value=Select(
     MouseAndKeyboard:float=Select!(Option 0:float=0, Option 1:float=1, Index:bool=@n8) -> (ReturnValue)
     Gamepad:float=0
     Touch:float=0
     Count:float=0
     Index=@n1
    ) -> (ReturnValue:float)
   )
   > SetVisibility(self=Get HoldProgressBar(), InVisibility=SelfHitTestInvisible)
   n9 = Get IsPromptIcon()
   > SetVisibility(
    self=Get HoldIcon()
    InVisibility=Select!(
     MouseAndKeyboard:ESlateVisibility Enum=Select!(Option 0:ESlateVisibility Enum=SelfHitTestInvisible, Option 1:ESlateVisibility Enum=Collapsed, Index:bool=@n9) -> (ReturnValue)
     Gamepad:ESlateVisibility Enum=Collapsed
     Touch:ESlateVisibility Enum=SelfHitTestInvisible
     Count:ESlateVisibility Enum=Collapsed
     Index:ECommonInputType Enum=@n1
    ) -> (ReturnValue:ESlateVisibility Enum)
   )
  ]}
 UpdateSecondText {nodes [
   UpdateSecondText() {is=title K2Node_FunctionEntry}
   > n1 = Get!(execute) -> (then, else, Block_SecondText:WBP UIKit Block Text Object Reference)
   > SetFont(self=@n1, Font=Get SecondFont() -> (SecondFont:Slate Font Info Structure))
   > SetTransformPolicy!(
    execute
    self:WBP UIKit Block Text Object Reference=@n1
    Transform Policy:ETextTransformPolicy Enum=Get SecondTextTransformPolicy() -> (SecondTextTransformPolicy:ETextTransformPolicy Enum=None)
   ) -> (then)
   > SetMobileFontSizeMultiplier(
    self=@n1
    MobileFontSizeMultiplier=Select(
     Option 0:double=Get MobileFontSizeMultiplier()
     Option 1:double=Get MobileFontSizeMultiplierOverride()
     Index=Get OverrideDefaultMobileFontSizeMultiplier()
    ) -> (ReturnValue:double)
   )
   > Set Padding(Padding=Get SecondTextPadding() -> (SecondTextPadding:Margin Structure), self=@n1)
   > SetText(self=@n1, Text=Get SecondText() -> (SecondText:Text))
   > SetJustification!(
    execute
    self:WBP UIKit Block Text Object Reference=@n1
    Justification:ETextJustify Enum=Get TextSecondJustification() -> (TextSecondJustification:ETextJustify Enum=Left)
   ) -> (then)
  ]}
 UpdateStatusIndicator {nodes [
   UpdateStatusIndicator() {is=title K2Node_FunctionEntry}
   > n1 = Get!(execute) -> (then, else, StatusIndicator:WBP UIKit Status Indicator Object Reference)
   > SetIndicatorType(self=@n1, IndicatorType=Get Type() -> (Type:E_UI_StatusIndicator_Type Enum=NewEnumerator0))
   > SetIndicatorConfiguration(self=@n1, InCount=0, IndicatorConfig=Get Configuration() -> (Configuration:E_UI_StatusIndicator_Configuration Enum=NewEnumerator0))
   > SetGridSlotAlignment(
    Widget=@n1
    Horizontal=Get StatusIndicatorHorizontalAlignment() -> (StatusIndicatorHorizontalAlignment=HAlign_Fill)
    Vertical=Get StatusIndicatorVerticalAlignment() -> (StatusIndicatorVerticalAlignment:EVerticalAlignment Enum=VAlign_Fill)
   )
   > SetGridSlotPadding(Widget=@n1, Padding=Get StatusIndicatorPadding() -> (StatusIndicatorPadding:Margin Structure))
   > SetGridSlotNudge(Widget=@n1, Nudge=Get StatusIndicatorNudge() -> (StatusIndicatorNudge:Vector 2D Structure))
   > UpdateSize!(
    execute
    self:WBP UIKit Status Indicator Base Object Reference=@n1
    IconSize:E_UI_StatusIndicator_Size Enum=Get Icon Size() -> (Icon Size:E_UI_StatusIndicator_Size Enum=NewEnumerator2)
   ) -> (then)
  ]}
 UpdateText {nodes [
   UpdateText() {is=title K2Node_FunctionEntry}
   > n1 = Get!(execute) -> (then, else, Block_Text:WBP UIKit Block Text Object Reference)
   > SetFont(self=@n1, Font=Get Font() -> (Font:Slate Font Info Structure))
   > SetTransformPolicy(self=@n1, Transform Policy=Get TextTransformPolicy() -> (TextTransformPolicy:ETextTransformPolicy Enum=None))
   > SetMobileFontSizeMultiplier(
    self=@n1
    MobileFontSizeMultiplier=Select(
     Option 0=Get MobileFontSizeMultiplier()
     Option 1=Get MobileFontSizeMultiplierOverride()
     Index=Get OverrideDefaultMobileFontSizeMultiplier()
    )
   )
   > SetTextPadding!(execute, self, TextPadding:Margin Structure=Get TextPadding()) -> (then) {is=title K2Node_CallFunction}
   > SetTextNudge!(execute, self, TextNudge:Vector 2D Structure=Get TextNudge()) -> (then) {is=title K2Node_CallFunction}
   > SetText(self=@n1, Text=Get Text!() -> (Text))
   > SetJustification(self=@n1, Justification=Get TextJustification() -> (TextJustification:ETextJustify Enum=Left))
   > SetTextOverflowPolicy(self:Self Object Reference, OverflowPolicy=Get TextOverflowPolicy() -> (TextOverflowPolicy:ETextOverflowPolicy Enum=Clip))
   > Call On Block Text Set!(execute:Exec, self:WBP UIKit Button Generic Base Object Reference) -> (then:Exec) {is=title K2Node_CallDelegate}
  ]}
]
variables [
 BackgroundClass {type=Class, default=/FortUILibrary/Systems/ModularButton/Blocks/WBP_UIKit_Block_BackgroundGeneric.WBP_UIKit_Block_BackgroundGeneric_C}
 BackgroundMaterial {type=MaterialInstance, default=/FortUILibrary/Systems/ModularButton/MaterialInstances/MI_UIKit_Color}
 Block_Background {type=WBP_UIKit_BlockBackground_Base_C, default=null}
 Block_Icon {type=WBP_UIKit_Block_Image_C, default=null}
 Block_SecondText {type=WBP_UIKit_Block_Text_C, default=null}
 Block_Text {type=WBP_UIKit_Block_Text_C, default=null}
 ButtonPadding {type=Margin, default=<ltrb 24 0 24 0>}
 ClickedSound {type=SoundBase, default=null}
 Configuration {type=E_UI_StatusIndicator_Configuration, default=Dot}
 ContentHorizontalAlignment {type=EHorizontalAlignment, default=HAlign_Center}
 ContentVerticalAlignment {type=EVerticalAlignment, default=VAlign_Center}
 CurrentHoldProgress {type=number}
 CurrentInputType {type=ECommonInputType, default=MouseAndKeyboard}
 EnableFocusedSelectedState {type=boolean, default=true}
 FadeInTime {type=number, default=0.1}
 FadeOutTime {type=number, default=0.1}
 Font {
  type=SlateFontInfo
  default {
   bForceMonospaced=false
   bMaterialIsStencil=false
   fontMaterial=/FortUILibrary/Systems/ModularButton/MaterialInstances/MI_UIKit_ColorText
   fontObject=&Burbank
   letterSpacing=48
   monospacedWidth=1
   size=21
   skewAmount=0
   typefaceFontName=BurbankRegular-Black
   outlineSettings {
    bApplyOutlineToDropShadows=false
    bMiteredCorners=false
    bSeparateFillAlpha=false
    outlineMaterial=null
    outlineSize=0
    outlineColor=<rgba 0 0 0 1>
   }
  }
 }
 HoldCompletedSound {
  type=SoundBase
  default=/Game/Sounds/Fort_UI_Sounds/FrontEnd/DefaultSounds/HoldableButtons/Holdable_Regular/MSS_UI_Button_Holdable_Regular_Complete
 }
 HoldPressedSound {
  type=SoundBase
  default=/Game/Sounds/Fort_UI_Sounds/FrontEnd/DefaultSounds/HoldableButtons/Holdable_Regular/MSS_UI_Button_Holdable_Regular_Press
 }
 HoldReleasedSound {
  type=SoundBase
  default=/Game/Sounds/Fort_UI_Sounds/FrontEnd/DefaultSounds/HoldableButtons/Holdable_Regular/MSS_UI_Button_Holdable_Regular_Release
 }
 HoldSound {
  type=SoundBase
  default=/Game/Sounds/Fort_UI_Sounds/FrontEnd/DefaultSounds/HoldableButtons/Holdable_Regular/MSS_UI_Button_Holdable_Regular_Hold_lp
 }
 HoldSoundAudioComponent {type=AudioComponent, default=null}
 HoveredSound {type=SoundBase, default=/Game/Sounds/Fort_UI_Sounds/FrontEnd/DefaultSounds/PressedButtons/MSS_UI_Button_Hover_Default}
 Icon {type=Texture2D}
 Icon Size {type=E_UI_StatusIndicator_Size, default=Default}
 IconClass {type=Class, default=&WBP_UIKit_Block_Image_C}
 IconHorizontalAlignment {type=EHorizontalAlignment, default=HAlign_Center}
 IconMaterial {type=MaterialInstance, default=/FortUILibrary/Systems/ModularButton/MaterialInstances/MI_UIKit_Icon}
 IconPadding {type=Margin, default=<ltrb 0 0 0 0>}
 IconSize {type=Vector2D, default=<32 32>}
 IconVerticalAlignment {type=EVerticalAlignment, default=VAlign_Center}
 InputActionProgressMaterial {type=MaterialInstanceDynamic}
 IsInstantSelection {type=boolean}
 IsStatusIndicatorInsideMainGrid {type=boolean}
 IsTextOutsideBackground {type=boolean}
 LastEnabledEvent {type=boolean}
 MinHeightOverride {type=integer}
 MinWidthOverride {type=integer}
 MobileButtonPadding {type=Margin, default=<ltrb 24 0 24 0>}
 MobileFontSizeMultiplier {type=number, default=1.3}
 MobileFontSizeMultiplierOverride {type=number, default=1}
 MobileMinHeight {type=integer}
 MobileMinWidth {type=integer}
 OnBlockTextSet {}
 OnHoldActionProgress {}
 OnHoldCompleted {}
 OnReleased {}
 OutlineMaterial {type=MaterialInstance, default=null}
 OverrideDefaultMobileFontSizeMultiplier {type=boolean}
 PressedSound {type=SoundBase, default=/Game/Sounds/Fort_UI_Sounds/FrontEnd/DefaultSounds/PressedButtons/MSS_UI_Button_Pressed_Default}
 PressedSoundAudioComponent {type=AudioComponent, default=null}
 PromptHorizontalAlignment {type=EHorizontalAlignment, default=HAlign_Left}
 PromptNudge {type=Vector2D, default=<-24 0>}
 PromptPadding {type=Margin, default=<ltrb 0 0 0 0>}
 PromptVerticalAlignment {type=EVerticalAlignment, default=VAlign_Center}
 SecondFont {
  type=SlateFontInfo
  default {
   bForceMonospaced=false
   bMaterialIsStencil=false
   fontMaterial=&MI_UIKit_ColorText
   fontObject=&Burbank
   letterSpacing=48
   monospacedWidth=1
   size=21
   skewAmount=0
   typefaceFontName=BurbankRegular-Black
   outlineSettings {
    bApplyOutlineToDropShadows=false
    bMiteredCorners=false
    bSeparateFillAlpha=false
    outlineMaterial=null
    outlineSize=0
    outlineColor=<rgba 0 0 0 1>
   }
  }
 }
 SecondText {type=string, default=Second}
 SecondTextBlockAlignment {type=E_UI_Alignment, default=Right}
 SecondTextClass {type=Class, default=/FortUILibrary/Systems/ModularButton/Blocks/WBP_UIKit_Block_Text.WBP_UIKit_Block_Text_C}
 SecondTextPadding {type=Margin, default=<ltrb 20 0 0 0>}
 SecondTextTransformPolicy {type=ETextTransformPolicy, default=ToUpper}
 ShouldPromptDirectionFollowCulture {type=boolean, default=true}
 ShouldPromptImpactLayout {type=boolean}
 ShouldTextImpactLayout {type=boolean, default=true}
 StatusIndicator {type=WBP_UIKit_StatusIndicator_C, default=null}
 StatusIndicatorClass {type=Class, default=/FortUILibrary/UIKit/StatusIndicator/WBP_UIKit_StatusIndicator.WBP_UIKit_StatusIndicator_C}
 StatusIndicatorHorizontalAlignment {type=EHorizontalAlignment, default=HAlign_Right}
 StatusIndicatorNudge {type=Vector2D, default=<0 0>}
 StatusIndicatorPadding {type=Margin, default=<ltrb 0 8 8 0>}
 StatusIndicatorVerticalAlignment {type=EVerticalAlignment, default=VAlign_Top}
 Text {type=string, default=generic BÄŞE}
 TextBlockAlignment {type=E_UI_Alignment, default=Right}
 TextClass {type=Class, default=&WBP_UIKit_Block_Text_C}
 TextJustification {type=ETextJustify, default=Center}
 TextNudge {type=Vector2D, default=<0 0>}
 TextOverflowPolicy {type=ETextOverflowPolicy, default=Clip}
 TextPadding {type=Margin, default=<ltrb 0 0 0 0>}
 TextSecondJustification {type=ETextJustify, default=Center}
 TextTransformPolicy {type=ETextTransformPolicy, default=ToUpper}
 Type {type=E_UI_StatusIndicator_Type, default=Attention}
 UseBackgroundBlock {type=boolean, default=true}
 UseDarkInputIcon {type=boolean}
 UseDisabledVisualsOnLocked {type=boolean}
 UseIconBlock {type=boolean, default=true}
 UseSecondTextBlock {type=boolean}
 UseStatusIndicator {type=boolean}
 UseTextBlock {type=boolean, default=true}
 VisualAttachmentContentAnchor {type=Vector2D, default=<0.5 0.5>}
 VisualAttachmentHorizontalAlignment {type=EHorizontalAlignment, default=HAlign_Center}
 VisualAttachmentVerticalAlignment {type=EVerticalAlignment, default=VAlign_Center}
]
widgets {
 info {inheritedWidgetCount=1, parentClass=/Script/FortniteUI.UIKitModularButton, rootWidgetClass=&GridPanel, widgetCount=11}
 widgets [
  Root {widgetClassPath=&GridPanel, bIsVariable=true, widget=Root, properties {columnFill [0,1,0], rowFill [0,1,0]}}
  Slot_Background {
   widgetClassPath=&NamedSlot
   bIsVariable=true
   parent=Root
   widget=Slot_Background
   slot {class=GridSlot, properties {column=1, row=1}}
  }
  Main {
   widgetClassPath=&GridPanel
   bIsVariable=true
   parent=Root
   widget=Main
   properties {columnFill [0,1,0], rowFill [0,1,0]}
   slot {class=GridSlot, properties {column=1, columnSpan=1, layer=10, row=1, rowSpan=1}}
  }
  Main_StackSize {
   widgetClassPath=&SizeBox
   bIsVariable=true
   parent=Main
   widget=Main_StackSize
   slot {class=GridSlot, properties {column=1, columnSpan=1, row=1, rowSpan=1}}
  }
  Main_Stack {
   widgetClassPath=&StackBox
   bIsVariable=true
   parent=Main_StackSize
   widget=Main_Stack
   properties {orientation=Orient_Vertical}
   slot {class=SizeBoxSlot}
  }
  VisualAttachment {
   widgetClassPath=&CommonVisualAttachment
   bIsVariable=true
   parent=Root
   widget=VisualAttachment
   properties {contentAnchor=<0.5 0.5>}
   slot {
    class=GridSlot
    properties {column=1, columnSpan=1, horizontalAlignment=HAlign_Center, layer=20, row=1, rowSpan=1, verticalAlignment=VAlign_Center}
   }
  }
  Slot_Main {
   widgetClassPath=&NamedSlot
   bIsVariable=true
   parent=Root
   widget=Slot_Main
   slot {class=GridSlot, properties {column=1, columnSpan=1, layer=40, row=1, rowSpan=1}}
  }
  Prompt {
   widgetClassPath=&GridPanel
   bIsVariable=true
   parent=Root
   widget=Prompt
   properties {visibility=Collapsed, columnFill [1], rowFill [1]}
   slot {
    class=GridSlot
    properties {column=1, columnSpan=1, horizontalAlignment=HAlign_Left, layer=20, row=1, rowSpan=1, verticalAlignment=VAlign_Center, nudge=<-24 0>}
   }
  }
  HoldProgressBar {
   widgetClassPath=&Image
   bIsVariable=true
   parent=Prompt
   widget=HoldProgressBar
   properties {
    visibility=Collapsed
    brush {
     bIsDynamicallyLoaded=false
     drawAs=Image
     imageType=NoImage
     mirroring=NoMirror
     resourceName=None
     resourceObject=/FortUILibrary/UIKit/Buttons/MaterialInstances/CTAButtons/MI_UIKit_InputActionProgress
     tiling=NoTile
     imageSize=<48 48>
     margin=<ltrb 0 0 0 0>
     outlineSettings {
      bUseBrushTransparency=false
      roundingType=HalfHeightRadius
      width=0
      color {colorUseRule=UseColor_Specified, specifiedColor=<rgba 0 0 0 0>}
      cornerRadii=<0 0 0 0>
     }
     tintColor {colorUseRule=UseColor_Specified, specifiedColor=<rgba 1 1 1 1>}
     uVRegion {bIsValid=false, max=<0 0>, min=<0 0>}
    }
   }
   slot {class=GridSlot, properties {horizontalAlignment=HAlign_Center, layer=10, verticalAlignment=VAlign_Center}}
  }
  HoldIcon {
   widgetClassPath=&Image
   bIsVariable=true
   parent=Prompt
   widget=HoldIcon
   properties {
    visibility=Collapsed
    brush {
     bIsDynamicallyLoaded=false
     drawAs=Image
     imageType=NoImage
     mirroring=NoMirror
     resourceName=None
     resourceObject=/FortUILibrary/Systems/ModularButton/MaterialInstances/MI_MousePressAndHold
     tiling=NoTile
     imageSize=<28 28>
     margin=<ltrb 0 0 0 0>
     outlineSettings {
      bUseBrushTransparency=false
      roundingType=HalfHeightRadius
      width=0
      color {colorUseRule=UseColor_Specified, specifiedColor=<rgba 0 0 0 0>}
      cornerRadii=<0 0 0 0>
     }
     tintColor {colorUseRule=UseColor_Specified, specifiedColor=<rgba 1 1 1 1>}
     uVRegion {bIsValid=false, max=<0 0>, min=<0 0>}
    }
   }
   slot {class=GridSlot, properties {horizontalAlignment=HAlign_Center, layer=30, verticalAlignment=VAlign_Center}}
  }
  InputActionWidget {
   widgetClassPath=&CommonActionWidget
   bInherited=true
   bIsVariable=true
   parent=Prompt
   widget=InputActionWidget
   properties {
    progressMaterialParam=Percentage
    visibility=HitTestInvisible
    progressMaterialBrush {
     bIsDynamicallyLoaded=false
     drawAs=Image
     imageType=NoImage
     mirroring=NoMirror
     resourceName=None
     resourceObject=/FortUILibrary/UIKit/Buttons/MaterialInstances/CTAButtons/MI_UIKit_EmptyProgress
     tiling=NoTile
     imageSize=<48 48>
     margin=<ltrb 0 0 0 0>
     outlineSettings {
      bUseBrushTransparency=false
      roundingType=HalfHeightRadius
      width=0
      color {colorUseRule=UseColor_Specified, specifiedColor=<rgba 0 0 0 0>}
      cornerRadii=<0 0 0 0>
     }
     tintColor {colorUseRule=UseColor_Specified, specifiedColor=<rgba 1 1 1 1>}
     uVRegion {bIsValid=false, max=<0 0>, min=<0 0>}
    }
   }
   slot {class=GridSlot, properties {horizontalAlignment=HAlign_Center, layer=40, verticalAlignment=VAlign_Center}}
  }
 ]
}
